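# Subsurface material for the skeletal prop of the HNY Good Riddance 2020/2021 cosmetic animation, using the Push Pin texture set.
# A normal-mapped Fresnel drives opacity and splits albedo between base color and tinted subsurface color.
# The Airout parameter scales animated pre-skinned-position noise that displaces vertices along normals and adds a crumpled normal.
# Vertex red blends this into a metallic variant tinted by a view-angle reflectance curve, without crumpling.
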Material /BRCosmetics/Animation/Props/HNY_Good_Riddance/Materials/M_HNY_Good_Riddance
graph {
 properties {
  bUseMaterialAttributes=true
  bUsedWithMeshDeformer=true
  bUsedWithSkeletalMesh=true
  pixelDepthOffsetMode=PDOM_Legacy
  shadingModel=MSM_Subsurface
  usedShadingModels=Subsurface
 }
 nodes [
  n1 = PreSkinnedPosition()
  n2 = ShaderStageSwitch(PixelShader=VertexInterpolator(VS=@n1).PS, VertexShader=@n1)
  Add(A=@n2)
  n3 = TextureSample() {texture=/BRCosmetics/Animation/Props/HNY_Good_Riddance/Textures/T_HNY_Good_Riddance_2020_2021_Push_Pin_D}
  n4 = TextureSample() {samplerType=SAMPLERTYPE_Normal, texture=/BRCosmetics/Animation/Props/HNY_Good_Riddance/Textures/T_HNY_Good_Riddance_2020_2021_Push_Pin_N}
  n5 = Fresnel(ExponentIn=Constant() {r=1}, Normal=Transform(None=@n4.RGB))
  n6 = OneMinus(None=@n5)
  n7 = TextureSample() {samplerType=SAMPLERTYPE_Masks, texture=/BRCosmetics/Animation/Props/HNY_Good_Riddance/Textures/T_HNY_Good_Riddance_2020_2021_Push_Pin_S}
  n8 = ScalarParameter() {parameterName=MinSpecular}
  n9 = ScalarParameter() {defaultValue=1, parameterName=MaxSpecular}
  n10 = ScalarParameter() {parameterName=MinRoughness}
  n11 = ScalarParameter() {defaultValue=0.75, parameterName=MaxRoughness}
  n12 = Multiply(A=Multiply(A=@n3.RGB, B=VectorParameter() {parameterName=SS_Tint, defaultValue=<rgba 0.838 0.579 0.388 1>}.RGB), B=@n5)
  n13 = ScalarParameter() {defaultValue=0.75, parameterName=SubsurfaceColorStrength}
  n14 = GetMaterialAttributes(None=MakeMaterialAttributes!(
    BaseColor=Multiply(A=@n3.RGB, B=@n6)
    Metallic=@n7.G
    Specular=LinearInterpolate(A=@n8, B=@n9, Alpha=@n7.R)
    Roughness=LinearInterpolate(A=@n10, B=@n11, Alpha=@n7.B)
    Anisotropy
    EmissiveColor
    Opacity=@n5
    OpacityMask
    Normal=@n4.RGB
    Tangent
    WorldPositionOffset
    SubsurfaceColor=Multiply(A=@n12, B=@n13)
    ClearCoat
    ClearCoatRoughness
    AmbientOcclusion
    Refraction
    CustomizedUVs_0
    CustomizedUVs_1
    CustomizedUVs_2
    CustomizedUVs_3
    CustomizedUVs_4
    CustomizedUVs_5
    CustomizedUVs_6
    CustomizedUVs_7
    PixelDepthOffset
    ShadingModel
    Displacement
   )) {attributeGetTypes [#1, #2, #3]}
  n15 = VertexNormalWS()
  n16 = TextureObject() {texture=/Game/Characters/Player/Male/Medium/Bodies/M_MED_Viper/Textures/Glitch/T_SoftNoise}
  n17 = VectorParameter() {parameterName=NoiseScale, defaultValue=<rgba 24 24 24 1>}
  n18 = Add(A=@n2, B=Multiply(A=Time(), B=VectorParameter() {parameterName=NoiseSpeed, defaultValue=<rgba 0.25 -0.3 0.1 1>}.RGB))
  n19 = ScalarParameter() {parameterName=Airout}
  n20 = Multiply(
   A=ComponentMask(None=MaterialFunctionCall(TextureObject=@n16, TextureSize=@n17.RGB, WorldPosition=@n18) {materialFunction=&WorldAlignedTexture}.XYZ Texture) {r=true}
   B=Multiply(A=ScalarParameter() {defaultValue=-3, parameterName=VertexCrunch_Strength}, B=@n19)
  )
  n21 = MaterialFunctionCall(
   Normal=MaterialFunctionCall(
    TextureObject=TextureObject() {samplerType=SAMPLERTYPE_Normal, texture=/Game/Global/Textures/Default/Noise/T_SwirlNoise_N}
    TextureSize=@n17.RGB
    WorldPosition=@n2
   ) {materialFunction=&WorldAlignedNormal}.XYZ Texture
   Flatness=LinearInterpolate(A=Constant() {r=1}, B=ScalarParameter() {defaultValue=0.25, parameterName=Crumpliness}, Alpha=@n19)
  ) {materialFunction=&FlattenNormal}
  n22 = CameraVectorWS()
  n23 = PixelNormalWS()
  n24 = CurveAtlasRowParameter(CurveTime=Clamp(None=DotProduct(A=@n22, B=@n23))) {
   atlas=&AthenaCurveAtlas01
   curve=/Game/Characters/Player/Female/Medium/Bodies/F_MED_Ms_Alpine/Textures/AluminiumreflectanceCurve
   defaultValue=189
   group=LUT
   parameterName=FresnelReflectance_LUT
  }
  n25 = ChannelMaskParameter(None=@n24) {group=LUT, parameterName=LUTChannel}
  n26 = ScalarParameter() {defaultValue=0.5, parameterName=2021_MaxRoughness}
  BlendMaterialAttributes(
   A=SetMaterialAttributes!(
    MaterialAttributes=@n14.MaterialAttributes
    World Position Offset=Multiply(A=@n15, B=@n20)
    Normal=MaterialFunctionCall(BaseNormal=@n14.Normal, AdditionalNormal=@n21.Result) {materialFunction=&BlendAngleCorrectedNormals}.Result
   ) {attributeSetTypes [#4, #1]}
   B=SetMaterialAttributes!(
    MaterialAttributes=@n14.MaterialAttributes
    Base Color=Multiply(A=@n14.Base Color, B=@n25.R)
    Metallic=Constant() {r=1}
    Roughness=LinearInterpolate(B=@n26, Alpha=@n14.Roughness)
   ) {attributeSetTypes [#2, #5, #3]}
   Alpha=VertexColor().R
  )
  ScalarParameter() {defaultValue=0.25, parameterName=NoiseSpeed}
 ]
}
