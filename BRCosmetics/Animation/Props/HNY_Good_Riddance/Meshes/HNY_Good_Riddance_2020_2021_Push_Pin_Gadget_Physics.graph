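# Physics asset for the push pin gadget prop in the HNY Good Riddance 2020-2021 cosmetic animation set.
# A single capsule body on the root bone, offset upward from the bone origin, is the prop's only collision shape.
# It has no constraints, so the asset defines no articulation between bones.
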PhysicsAsset /BRCosmetics/Animation/Props/HNY_Good_Riddance/Meshes/HNY_Good_Riddance_2020_2021_Push_Pin_Gadget_Physics
bodies {root [{shapeType=Capsule, length=90, radius=31.255, center=<-0.0003 0 72.2839>, rotation=<rot 0 -90 0>}]}
constraints []
